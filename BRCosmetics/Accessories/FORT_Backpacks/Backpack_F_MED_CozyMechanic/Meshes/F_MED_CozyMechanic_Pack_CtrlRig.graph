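# Control Rig for the CozyMechanic backpack cosmetic, with procedurally built controls for posing its root, pack, left/right wing and fan bones.
# Construction imports the skeleton, tags those bones with shape, color, offset and parent metadata, then spawns a control per imported item from that metadata.
# Forward solve copies control transforms onto bones and backward solve copies bones onto controls, pairing both arrays by index.
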ControlRigBlueprint /BRCosmetics/Accessories/FORT_Backpacks/Backpack_F_MED_CozyMechanic/Meshes/F_MED_CozyMechanic_Pack_CtrlRig
graphs [
 New Function_ContainedGraph {nodes [
   n1 = RigVMFunctionEntryNode()
   RigVMFunction_Sequence = &RigVMFunction_Sequence(ExecuteContext=@n1.Argument)
   n2 = &RigUnit_HierarchyImportFromSkeleton(NameSpace=None, bIncludeCurves=False, bIncludeMeshSockets=False, bIncludeVirtualBones=True, ExecutePin=@RigVMFunction_Sequence.A)
   > RigVMVariableNode(Value=@n2.Items) {variable=Skeleton Array}
   RigVMFunctionReturnNode(Argument=@RigVMFunction_Sequence.B)
  ]}
 New Function_ContainedGraph {nodes [
   n1 = RigVMFunctionEntryNode()
   RigVMFunction_Sequence = &RigVMFunction_Sequence(ExecuteContext=@n1.Argument)
   For_Each = RigVMDispatchNode(ExecuteContext=@RigVMFunction_Sequence.A, Array=RigVMVariableNode() {variable=Skeleton Array}.Value) {template="DISPATCH_RigVMDispatch_ArrayIterator(in Array,out Element,out Index,out Count,out Ratio)"}
   > RigVMVariableNode(Value=@For_Each.Element) {variable=local element}
   n2 = RigVMVariableNode() {variable=local element}
   Get_Metadata = RigVMDispatchNode(Default=(Rotation=(X=0,Y=0,Z=0,W=1),Translation=(X=0,Y=0,Z=0),Scale3D=(X=1,Y=1,Z=1)), Name=Offset, NameSpace=Self, Item=@n2.Value) {template="DISPATCH_RigDispatch_GetMetadata(in Item,in Name,in NameSpace,in Default,out Value,out Found)"}
   Get_Metadata_1 = RigVMDispatchNode(Default=(R=1,G=1,B=1,A=1), Name=Color, NameSpace=Self, Item=@n2.Value) {template="DISPATCH_RigDispatch_GetMetadata(in Item,in Name,in NameSpace,in Default,out Value,out Found)"}
   Get_Metadata_2 = RigVMDispatchNode(Default=Default, Name=Shape, NameSpace=Self, Item=@n2.Value) {template="DISPATCH_RigDispatch_GetMetadata(in Item,in Name,in NameSpace,in Default,out Value,out Found)"}
   Get_Metadata_3 = RigVMDispatchNode(Default=(Type=Bone,Name="None"), Name=Parent, NameSpace=Self, Item=@n2.Value) {template="DISPATCH_RigDispatch_GetMetadata(in Item,in Name,in NameSpace,in Default,out Value,out Found)"}
   > SpawnControl = /Script/ControlRig.RigUnit_HierarchyAddControlTransform(
    OffsetSpace=GlobalSpace
    OffsetTransform=(Rotation=(X=0,Y=0,Z=0,W=1),Translation=(X=0,Y=0,Z=0),Scale3D=(X=1,Y=1,Z=1))
    Settings="(InitialSpace=LocalSpace,bUsePreferredRotationOrder=False,PreferredRotationOrder=YZX,Limits=(LimitTranslationX=(bMinimum=False,bMaximum=False),LimitTranslationY=(bMinimum=False,bMaximum=False),LimitTranslationZ=(bMinimum=False,bMaximum=False),LimitPitch=(b ...(truncated)"
    InitialValue=&RigUnit_GetTransform(bInitial=true, Item=@n2.Value).Transform
    Settings.Shape.Transform=@Get_Metadata.Value
    Settings.Shape.Color=@Get_Metadata_1.Value
    Settings.Shape.Name=@Get_Metadata_2.Value
    Parent=@Get_Metadata_3.Value
    Name=RigVMVariableNode() {variable=local element}.Value.Name
   ) {
    is=struct RigVMUnitNode
    template="SpawnControl::Execute(in InitialValue,in Settings,in OffsetTransform,in OffsetSpace,in Parent,in Name,out Item)"
   }
   > RigVMDispatchNode(Array=RigVMVariableNode() {variable=Control Array}.Value, Element=@SpawnControl.Item) {template="DISPATCH_RigVMDispatch_ArrayAdd(io Array,in Element,out Index)"}
   RigVMFunctionReturnNode(Argument=@RigVMFunction_Sequence.B)
  ]}
 New Function_ContainedGraph {nodes [
   RigVMFunctionEntryNode()
   > RigVMFunction_Sequence = RigVMAggregateNode()
   RigVMFunctionReturnNode(ExecuteContext=@RigVMFunction_Sequence.C)
   Set_Metadata_3 = RigVMDispatchNode(
    Name=Color
    NameSpace=Self
    Value=RigVMVariableNode() {variable=Colour}.Value
    Item=RigVMVariableNode() {variable=Bone}.Value
    ExecuteContext=@RigVMFunction_Sequence.A
   ) {template="DISPATCH_RigDispatch_SetMetadata(in Item,in Name,in NameSpace,in Value,out Success)"}
   > Set_Metadata_1 = RigVMDispatchNode(Name=Shape, NameSpace=Self, Item=RigVMVariableNode() {variable=Bone}.Value, Value=RigVMVariableNode() {variable=Shape}.Value) {template="DISPATCH_RigDispatch_SetMetadata(in Item,in Name,in NameSpace,in Value,out Success)"}
   Set_Metadata_2 = RigVMDispatchNode(
    Name=Parent
    NameSpace=Self
    Item=RigVMVariableNode() {variable=Bone}.Value
    Value=RigVMVariableNode() {variable=Parent}.Value
    ExecuteContext=@RigVMFunction_Sequence.B
   ) {template="DISPATCH_RigDispatch_SetMetadata(in Item,in Name,in NameSpace,in Value,out Success)"}
   > Set_Metadata_4 = RigVMDispatchNode(Name=Offset, NameSpace=Self, Value=RigVMVariableNode() {variable=Offset}.Value, Item=RigVMVariableNode() {variable=Bone}.Value) {template="DISPATCH_RigDispatch_SetMetadata(in Item,in Name,in NameSpace,in Value,out Success)"}
  ]}
 RigVMFunctionLibrary {nodes [
   Create Controls = RigVMCollapseNode(Argument=())
   Set Bone Metadata = RigVMCollapseNode(
    Bone=(Type=None,Name="None")
    Colour=(R=0,G=0,B=0,A=0)
    ExecuteContext=()
    Offset=(Rotation=(X=0,Y=0,Z=0,W=1),Translation=(X=0,Y=0,Z=0),Scale3D=(X=1,Y=1,Z=1))
    Parent=(Type=None,Name="None")
   )
   Update Skeleton = RigVMCollapseNode(Argument=())
  ]}
 RigVMFunction_Sequence_1_ContainedGraph {nodes [
   n1 = RigVMFunctionEntryNode()
   RigVMFunction_Sequence = &RigVMFunction_Sequence(ExecuteContext=@n1.ExecuteContext)
   n2 = &RigVMFunction_Sequence(ExecuteContext=@RigVMFunction_Sequence.B)
   n3 = &RigVMFunction_Sequence(ExecuteContext=@n2.B)
   RigVMFunctionReturnNode(A=@RigVMFunction_Sequence.A, B=@n2.A, C=@n3.A, D=@n3.B)
  ]}
 RigVMFunction_Sequence_1_ContainedGraph {nodes [
   n1 = RigVMFunctionEntryNode()
   RigVMFunction_Sequence = &RigVMFunction_Sequence(ExecuteContext=@n1.ExecuteContext)
   n2 = &RigVMFunction_Sequence(ExecuteContext=@RigVMFunction_Sequence.B)
   RigVMFunctionReturnNode(A=@RigVMFunction_Sequence.A, B=@n2.A, C=@n2.B)
  ]}
 RigVMFunction_Sequence_1_ContainedGraph {nodes [
   n1 = RigVMFunctionEntryNode()
   RigVMFunction_Sequence = &RigVMFunction_Sequence(ExecuteContext=@n1.ExecuteContext)
   n2 = &RigVMFunction_Sequence(ExecuteContext=@RigVMFunction_Sequence.B)
   RigVMFunctionReturnNode(A=@RigVMFunction_Sequence.A, B=@n2.A, C=@n2.B)
  ]}
 RigVMModel {nodes [
   RigUnit_BeginExecution = &RigUnit_BeginExecution()
   For_Each = RigVMDispatchNode(ExecuteContext=@RigUnit_BeginExecution.ExecutePin, Array=RigVMVariableNode() {variable=Skeleton Array}.Value) {template="DISPATCH_RigVMDispatch_ArrayIterator(in Array,out Element,out Index,out Count,out Ratio)"}
   n1 = RigVMVariableNode() {variable=Control Array}
   > &RigUnit_SetTransform(
    Weight=1
    bInitial=False
    Item=@For_Each.Element
    Value=&RigUnit_GetTransform(
     bInitial=False
     Item=RigVMDispatchNode(Index=@For_Each.Index, Array=@n1.Value) {template="DISPATCH_RigVMDispatch_ArrayGetAtIndex(in Array,in Index,out Element)"}.Element
    ).Transform
   )
  ]}
 RigVMModel Construction Event {nodes [
   n1 = &RigUnit_PrepareForExecution()
   RigVMFunction_Sequence = RigVMAggregateNode(ExecuteContext=@n1.ExecutePin)
   RigVMDispatchNode(ExecuteContext=@RigVMFunction_Sequence.A, Array=RigVMVariableNode() {variable=Control Array}.Value) {template="DISPATCH_RigVMDispatch_ArrayReset(io Array)"}
   RigVMFunction_Sequence_1 = RigVMAggregateNode(ExecuteContext=@RigVMFunction_Sequence.C)
   Set Bone Metadata = RigVMFunctionReferenceNode(
    Bone=(Type=Bone,Name="root")
    Colour=(R=0,G=1,B=0,A=1)
    Offset=(Rotation=(X=0,Y=0,Z=0,W=1),Translation=(X=0,Y=0,Z=0),Scale3D=(X=7,Y=7,Z=7))
    Parent=(Type=None,Name="None")
    Shape=Circle_Thin
    ExecuteContext=@RigVMFunction_Sequence_1.A
   )
   > Set Bone Metadata_1 = RigVMFunctionReferenceNode(
    Bone=(Type=Bone,Name="pack")
    Colour=(R=0,G=1,B=0,A=1)
    Offset=(Rotation=(X=0,Y=0,Z=0,W=1),Translation=(X=0,Y=0,Z=0),Scale3D=(X=2,Y=1,Z=3))
    Parent=(Type=Control,Name="root")
    Shape=Box_Thin
   )
   Set Bone Metadata_2 = RigVMFunctionReferenceNode(
    Bone=(Type=Bone,Name="wing_l_01")
    Colour=(R=1,G=0,B=0,A=1)
    Offset=(Rotation=(X=0,Y=0,Z=0,W=1),Translation=(X=5,Y=0,Z=-5),Scale3D=(X=2,Y=0.5,Z=3))
    Parent=(Type=Control,Name="pack")
    Shape=Box_Thick
    ExecuteContext=@RigVMFunction_Sequence_1.B
   )
   > Set Bone Metadata_5 = RigVMFunctionReferenceNode(
    Bone=(Type=Bone,Name="fan_l_01")
    Colour=(R=1,G=0,B=0,A=1)
    Offset=(Rotation=(X=0,Y=0,Z=0,W=1),Translation=(X=0,Y=0,Z=0),Scale3D=(X=0.5,Y=1,Z=1))
    Parent=(Type=Control,Name="wing_l_01")
    Shape=Diamond_Thick
   )
   Set Bone Metadata_3 = RigVMFunctionReferenceNode(
    Bone=(Type=Bone,Name="wing_r_01")
    Colour=(R=0,G=0,B=1,A=1)
    Offset=(Rotation=(X=0,Y=0,Z=0,W=1),Translation=(X=-5,Y=0,Z=5),Scale3D=(X=2,Y=1,Z=3))
    Parent=(Type=Control,Name="pack")
    Shape=Box_Thick
    ExecuteContext=@RigVMFunction_Sequence_1.C
   )
   > Set Bone Metadata_4 = RigVMFunctionReferenceNode(
    Bone=(Type=Bone,Name="fan_r_01")
    Colour=(R=0,G=0,B=1,A=1)
    Offset=(Rotation=(X=0,Y=0,Z=0,W=1),Translation=(X=0,Y=0,Z=0),Scale3D=(X=0.5,Y=1,Z=1))
    Parent=(Type=Control,Name="wing_r_01")
    Shape=Diamond_Thick
   )
   Create Controls = RigVMFunctionReferenceNode(Argument=@RigVMFunction_Sequence.D)
   Update Skeleton = RigVMFunctionReferenceNode(Argument=@RigVMFunction_Sequence.B)
  ]}
 RigVMModel Rig {nodes [
   n1 = /Script/ControlRig.RigUnit_InverseExecution() {is=struct RigVMUnitNode}
   For_Each_1 = RigVMDispatchNode(ExecuteContext=@n1.ExecutePin, Array=RigVMVariableNode() {variable=Control Array}.Value) {template="DISPATCH_RigVMDispatch_ArrayIterator(in Array,out Element,out Index,out Count,out Ratio)"}
   n2 = RigVMVariableNode() {variable=Skeleton Array}
   > &RigUnit_SetTransform(
    Weight=1
    bInitial=False
    Item=@For_Each_1.Element
    Value=&RigUnit_GetTransform(
     bInitial=False
     Item=RigVMDispatchNode(Index=@For_Each_1.Index, Array=@n2.Value) {template="DISPATCH_RigVMDispatch_ArrayGetAtIndex(in Array,in Index,out Element)"}.Element
    ).Transform
   )
  ]}
]
hierarchy [NewControl {controlType=EULER_TRANSFORM, type=Control, loc=<0 0 0>, rot=<0 0 0 1>}]
variables [
 Control Array {type="TArray<FRigElementKey>"}
 Skeleton Array {type="TArray<FRigElementKey>"}
]
