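# Niagara system for the Cozy Mechanic backpack cosmetic's reactive effects, driven by a reactivity time value.
# Three burst emitters follow sockets: Fans places meshes that cross-fade with reactivity, while EyeGlow and InnerGlow place sprite glows that brighten with it.
# Glow emitters pass socket-rotated, mirrored direction vectors to dynamic material parameters for directional capsule glows.
# Variant selectors recolor the effects, with front-end and time-of-day scalars adjusting glow intensity.
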NiagaraSystem /BRCosmetics/Accessories/FORT_Backpacks/Backpack_F_MED_CozyMechanic/VFX/NS_Backpack_CozyMechanic_Reactivity
dependencies {
 usedDataInterfaces [&NiagaraDataInterfacePlatformSet, &NiagaraDataInterfaceCurve, &NiagaraDataInterfaceSocketReader]
 usedDynamicInputs [&MakeCustomFloatFromBool]
 usedModules [
  &self:SystemSpawnScript.SystemScriptSource.SystemScriptGraph.NiagaraNodeAssignment_3.SetVariables_#1
  &SystemState
  &FN_VariantManager
  &self:SystemSpawnScript.SystemScriptSource.SystemScriptGraph.NiagaraNodeAssignment_0.SetVariables_#2
  &self:SystemSpawnScript.SystemScriptSource.SystemScriptGraph.NiagaraNodeAssignment_1.SetVariables_#3
  &self:SystemSpawnScript.SystemScriptSource.SystemScriptGraph.NiagaraNodeAssignment_2.SetVariables_#4
 ]
}
emitters {
 EyeGlow {
  emitterUpdateScript [
   &EmitterState(
    Inactive Response=::NewEnumerator1
    Life Cycle Mode=::NewEnumerator1
    Loop Behavior=::NewEnumerator1
    Loop Duration Mode=::NewEnumerator1
    UseLoopDelay=false
   )
   &SpawnBurst_Instantaneous(Spawn Count=2, Use Spawn Probability=false, UseLoopCountLimit=true) {module=SpawnBurst_Instantaneous001}
  ]
  particleSpawnScript [
   &FN_SetSpawnGroup(Number of Spawn Groups=2)
   &InitializeParticle(
    Color=&VariantColorSelector(Variant Amount=1, Original Color=<rgba 0.4713 2.5 0 1>, Variant 1 Color=<rgba 2.5 0.3584 0 1>, Variant 2 Color=<rgba 0.8333 0 2.5 1>)
    Lifetime=$System.ReactivityDuration
    Sprite Size Mode=::NewEnumerator3
    Uniform Sprite Size=5
    UsePositionOffset=false
    Write Material Random=true
   ) {module=InitializeParticle001}
  ]
  particleUpdateScript [
   &ParticleState(Kill Particles When Lifetime Has Elapsed=false)
   &SocketLocation(
    Orientation Sampling=::NewEnumerator1
    SamplePosOffset=false
    Socket Index=&MergeHelper_IntSwitch(0=2, 1=3, 2=?, 3=?, 4=?, 5=?, Index=$Particles.SpawnGroup, Num=::NewEnumerator0)
    Socket Reader=$System.Socket Reader
    Socket Sampling Mode=::NewEnumerator3
    VelocitySpeedLimit=false
   )
   &ScaleColor(
    Scale RGB=&VectorFromFloat(Value=&Multiply_Float(
      A=&Multiply_Float(
       A=&Lerp_Float(A=0.7, B=1.2, Alpha=$System.ReactivityTime)
       B=&Waveform("Waveform [1]"=::NewEnumerator9, "[1] Frequency"=6, "[1] Amplitude Min/Max"=<0.75 1>)
      )
      B=&Multiply_Float(
       A=&Multiply_Float(
        A=$System.SystemFade
        B=&FN_FrontEndFloatModifier(
         Gameplay=1
         Front End=/BRCosmetics/Weapons/FORT_Melee/Pickaxe_GateHound/FX/VariantManager/VariantFloatSelector(
          Original Float:NiagaraFloat=0.7
          Variant 1 Float:NiagaraFloat=1
          Variant 2 Float:NiagaraFloat=1
          Variant 3 Float:NiagaraFloat=?
          Variant Amount:NiagaraInt32=1
         )
        )
       )
       B=&FNTimeOfDayScalar(Night Scale=0.875, Morning Scale=0.9, Evening Scale=0.935)
      )
     ))
    ScaleA=true
    ScaleRGB=true
   )
   /SetVariables(Particles.CapsuleGlowDirection:Vector3f=&TransformVector(
     Vector=&Multiply_Vector(
      A=&ReturnMeshOrientationAxis(
       Mesh Orientation Axis=::XAxis
       Mesh Orientation Quat=&MakeQuaternion(
        Quaternion From=1
        Axis=<1 0 0>
        Angle=0
        Yaw=47
        Pitch=26
        From=?
        To=?
        Angle Type=::NewEnumerator0
        Coordinate Space=::Local
        Normalize Vectors=false
       )
      )
      B=&MergeHelper_VectorSwitch(0=<1 1 1>, 1=<-1 1 1>, 2=?, 3=?, Index=$Particles.SpawnGroup, Num=::NewEnumerator0)
     )
     Destination Space=::World
    )) {isSetParameters=true, module=SetVariables_#5}
   &DynamicMaterialParameters(
    Index 0 Param 1=&MakeFloatFromVector(
     Vector=&RotateVectorByQuaternion(Vector To Rotate=$Particles.CapsuleGlowDirection, Quaternion=$Output.SocketLocation.SampledRotation)
     Channel=::NewEnumerator0
    )
    Index 0 Param 2=&MakeFloatFromVector(
     Vector=&RotateVectorByQuaternion(Vector To Rotate=$Particles.CapsuleGlowDirection, Quaternion=$Output.SocketLocation.SampledRotation)
     Channel=::NewEnumerator1
    )
    Index 0 Param 3=&MakeFloatFromVector(
     Vector=&RotateVectorByQuaternion(Vector To Rotate=$Particles.CapsuleGlowDirection, Quaternion=$Output.SocketLocation.SampledRotation)
     Channel=::NewEnumerator2
    )
    Index 0 Param 4=5.5
    Module.Index0Float0Write=true
    Module.Index0Float1Write=true
    Module.Index0Float2Write=true
    Module.Index0Float3Write=true
   )
   /SetVariables(Particles.DecalColor:LinearColor=&Multiply_LinearColorByFloat(Linear Color=$Particles.Color, Float=&FN_FrontEndFloatModifier(Gameplay=0.225, Front End=0.2))) {isSetParameters=true}
   &FN_Sprite_Attributes(
    Camera Offset=3
    Color=&MultiplyLinearColor(
     Linear Color=$Particles.Color
     Scale Factor=&MakeVector4FromVectorAndFloat("Vector (RGB)"=<1 1 1>, "Float (Alpha)"=&FN_FrontEndFloatModifier(Gameplay=0.45, Front End=1))
    )
    Create Camera Offset Attribute=true
    Create Color Attribute=true
    Create Position Attribute=true
    Create SpriteSize Attribute=true
    Position:NiagaraPosition=$Output.SocketLocation.SampledPosition
    SpriteSize=&MergeHelper_Vector2Switch(0=<4.5 4.5>, 1=<4.5 4.5>, 2=?, 3=?, 4=?, Index=$Particles.SpawnGroup, Num=::NewEnumerator0)
    Write to Module Namespace=true
   )
  ]
  rendererClasses [&NiagaraSpriteRendererProperties]
  renderers [
   {rendererClass=&NiagaraSpriteRendererProperties}
   {rendererClass=&NiagaraSpriteRendererProperties, rendererIndex=1}
  ]
 }
 Fans {
  emitterUpdateScript [
   &EmitterState(
    Inactive Response=::NewEnumerator1
    Life Cycle Mode=::NewEnumerator1
    Loop Behavior=::NewEnumerator1
    Loop Duration Mode=::NewEnumerator1
    UseLoopDelay=false
   )
   &SpawnBurst_Instantaneous(Spawn Count=4, Use Spawn Probability=false, UseLoopCountLimit=true)
   /SetVariables(Emitter.BlurGlow Color:LinearColor=&VariantColorSelector(Variant Amount=1, Original Color=<rgba 0.6576 5 0 1>, Variant 1 Color=<rgba 5 0.3333 0 1>, Variant 2 Color=<rgba 1.3333 0 5 1>)) {isSetParameters=true}
  ]
  particleSpawnScript [
   &FN_SetSpawnGroup(Number of Spawn Groups=2)
   &PartitionParticles(Number Of Partitions=4)
   &InitializeParticle(
    Color=&LerpLinearColors(
     StartColor=<rgba 0.2087 0.2782 0.5 3>
     EndColor=<rgba 0.2389 0.242 0.252 2>
     LerpFactor=&FN_PickFloatDependOnPlatform(TargetPlatform=DataInterface() {propertyValues {Platforms {qualityLevelMask=-29}}}, TargetPlatformActive=1, TargetPlatformNoActive=0)
    )
    Lifetime=$System.ReactivityDuration
    Mesh Scale Mode=::NewEnumerator3
    UsePositionOffset=false
    Write Material Random=true
   )
   &InitialMeshOrientation(Mesh Orientation Mode=::NewEnumerator2, Orientation Vector=<0 1 0>, Use Rotation Vector=true)
  ]
  particleUpdateScript [
   &ParticleState()
   &SocketLocation(
    SamplePosOffset=true
    Sampled Position Offset=&MergeHelper_VectorSwitch(0=<-1 0 0>, 1=<-2.1793 0 0>, 2=<1.45 0 0>, 3=<0.4 0 0>, Index=$Particles.PartitionParticles.Partition, Num=::NewEnumerator2)
    Socket Index=$Particles.SpawnGroup
    Socket Reader=$System.Socket Reader
    Socket Sampling Mode=::NewEnumerator3
    VelocitySpeedLimit=false
   )
   &ScaleColor(
    Scale Alpha=&FloatFromCurve(FloatCurve=DataInterface() {propertyValues {Curve {keys [<at 0.5 1 RCIM_Cubic>,<at 1 0 RCIM_Cubic>]}}}, Scale Curve=$System.ReactivityTime)
    ScaleA=true
    ScaleRGB=true
   )
   &DynamicMaterialParameters(
    Index 0 Param 1=0.7
    Index 0 Param 2=3
    Index 0 Param 3=1
    Index 0 Param 4=1
    Module.Index0Float0Write=true
    Module.Index0Float1Write=true
    Module.Index0Float2Write=true
    Module.Index0Float3Write=true
   )
   &FN_Mesh_Attributes(
    Color=&Multiply_LinearColorByFloat(
     Linear Color=&VariantColorSelector(
      Variant Amount=1
      Original Color=<rgba 0.0214 0.15 0.0523 0.1>
      Variant 1 Color=<rgba 0.487 0.0974 0.0696 0.1>
      Variant 2 Color=<rgba 0.4771 0.287 1 0.2>
     )
     Float=$System.ReactivityTime
    )
    Create Color Attribute=true
    Write to Module Namespace=true
   )
   &FN_Sprite_Attributes(
    Camera Offset=3
    Color=&Multiply_LinearColorByFloat(
     Linear Color=&VariantColorSelector(
      Variant Amount=1
      Original Color=<rgba 0.0814 0.3 0 0.1>
      Variant 1 Color=<rgba 0.3467 0.0497 0 0.1255>
      Variant 2 Color=<rgba 0.269 0.0889 0.3652 0.7739>
     )
     Float=$System.ReactivityTime
    )
    Create Camera Offset Attribute=true
    Create Color Attribute=true
    Write to Module Namespace=true
   )
   /SetVariables(Particles.DynamicMaterialParametersAlt:Vector4f=&MakeVector4(X=0.7, Y=3.5, Z=0, W=0)) {isSetParameters=true}
   &SolveForcesAndVelocity(Clamp Velocity=false, Limit Acceleration=false)
  ]
  rendererClasses [&NiagaraMeshRendererProperties]
  renderers [
   {rendererClass=&NiagaraMeshRendererProperties, rendererIndex=0}
   {rendererClass=&NiagaraMeshRendererProperties, rendererIndex=1}
   {rendererClass=&NiagaraMeshRendererProperties, rendererIndex=2}
  ]
 }
 InnerGlow {
  emitterUpdateScript [
   &EmitterState(
    Inactive Response=::NewEnumerator1
    Life Cycle Mode=::NewEnumerator1
    Loop Behavior=::NewEnumerator1
    Loop Duration Mode=::NewEnumerator1
    UseLoopDelay=false
   )
   &SpawnBurst_Instantaneous(Spawn Count=6, Use Spawn Probability=false, UseLoopCountLimit=true) {module=SpawnBurst_Instantaneous001}
  ]
  particleSpawnScript [
   &FN_SetSpawnGroup(Number of Spawn Groups=6)
   &InitializeParticle(
    Color=&VariantColorSelector(Variant Amount=1, Original Color=<rgba 0.2828 1.5 0 1>, Variant 1 Color=<rgba 1.5 0.215 0 1>, Variant 2 Color=<rgba 0.5 0 1.5 1>)
    Lifetime=$System.ReactivityDuration
    Sprite Size Mode=::NewEnumerator3
    Uniform Sprite Size=&MergeHelper_FloatSwitch(0=4, 1=5, 2=4, 3=4, 4=5, 5=4, Index=$Particles.SpawnGroup, Num=::NewEnumerator4)
    UsePositionOffset=false
    Write Material Random=true
   ) {module=InitializeParticle001}
  ]
  particleUpdateScript [
   &ParticleState()
   &SocketLocation(
    Orientation Sampling=::NewEnumerator1
    SamplePosOffset=false
    Socket Index=&MergeHelper_IntSwitch(0=8, 1=6, 2=4, 3=9, 4=7, 5=5, Index=$Particles.SpawnGroup, Num=::NewEnumerator4)
    Socket Reader=$System.Socket Reader
    Socket Sampling Mode=::NewEnumerator3
    VelocitySpeedLimit=false
   )
   &ScaleColor(
    Scale RGB=&VectorFromFloat(Value=&Multiply_Float(
      A=$System.ReactivityTime
      B=&FNTimeOfDayScalar(Night Scale=0.875, Morning Scale=0.9, Evening Scale=0.935)
     ))
    ScaleA=true
    ScaleRGB=true
   )
   /SetVariables(Particles.CapsuleGlowDirection:Vector3f=&TransformVector(
     Vector=&Multiply_Vector(
      A=&MergeHelper_VectorSwitch(
       0=&ReturnMeshOrientationAxis(
        Mesh Orientation Axis=::XAxis
        Mesh Orientation Quat=&MakeQuaternion(
         Quaternion From=1
         Axis=<1 0 0>
         Angle=0
         Yaw=58.8706
         Pitch=42
         From=?
         To=?
         Angle Type=::NewEnumerator0
         Coordinate Space=::Local
         Normalize Vectors=false
        )
       )
       1=&ReturnMeshOrientationAxis(
        Mesh Orientation Axis=::XAxis
        Mesh Orientation Quat=&MakeQuaternion(
         Quaternion From=1
         Axis=<1 0 0>
         Angle=0
         Yaw=51.6835
         Pitch=32
         From=?
         To=?
         Angle Type=::NewEnumerator0
         Coordinate Space=::Local
         Normalize Vectors=false
        )
       )
       2=&ReturnMeshOrientationAxis(
        Mesh Orientation Axis=::XAxis
        Mesh Orientation Quat=&MakeQuaternion(
         Quaternion From=1
         Axis=<1 0 0>
         Angle=0
         Yaw=-30
         Pitch=70
         From=?
         To=?
         Angle Type=::NewEnumerator0
         Coordinate Space=::Local
         Normalize Vectors=false
        )
       )
       3=&ReturnMeshOrientationAxis(
        Mesh Orientation Axis=::XAxis
        Mesh Orientation Quat=&MakeQuaternion(
         Quaternion From=1
         Axis=<1 0 0>
         Angle=0
         Yaw=58.8706
         Pitch=42
         From=?
         To=?
         Angle Type=::NewEnumerator0
         Coordinate Space=::Local
         Normalize Vectors=false
        )
       )
       4=&ReturnMeshOrientationAxis(
        Mesh Orientation Axis=::XAxis
        Mesh Orientation Quat=&MakeQuaternion(
         Quaternion From=1
         Axis=<1 0 0>
         Angle=0
         Yaw=51.6835
         Pitch=32
         From=?
         To=?
         Angle Type=::NewEnumerator0
         Coordinate Space=::Local
         Normalize Vectors=false
        )
       )
       5=&ReturnMeshOrientationAxis(
        Mesh Orientation Axis=::XAxis
        Mesh Orientation Quat=&MakeQuaternion(
         Quaternion From=1
         Axis=<1 0 0>
         Angle=0
         Yaw=-30
         Pitch=70
         From=?
         To=?
         Angle Type=::NewEnumerator0
         Coordinate Space=::Local
         Normalize Vectors=false
        )
       )
       Index=$Particles.SpawnGroup
       Num=::NewEnumerator4
      )
      B=&MergeHelper_VectorSwitch(0=<1 1 1>, 1=<1 1 1>, 2=<1 1 1>, 3=<-1 1 1>, 4=<-1 1 1>, 5=<-1 1 1>, Index=$Particles.SpawnGroup, Num=::NewEnumerator4)
     )
     Destination Space=::World
    )) {isSetParameters=true, module=SetVariables_#5}
   &DynamicMaterialParameters(
    Index 0 Param 1=&MakeFloatFromVector(
     Vector=&RotateVectorByQuaternion(Vector To Rotate=$Particles.CapsuleGlowDirection, Quaternion=$Output.SocketLocation.SampledRotation)
     Channel=::NewEnumerator0
    )
    Index 0 Param 2=&MakeFloatFromVector(
     Vector=&RotateVectorByQuaternion(Vector To Rotate=$Particles.CapsuleGlowDirection, Quaternion=$Output.SocketLocation.SampledRotation)
     Channel=::NewEnumerator1
    )
    Index 0 Param 3=&MakeFloatFromVector(
     Vector=&RotateVectorByQuaternion(Vector To Rotate=$Particles.CapsuleGlowDirection, Quaternion=$Output.SocketLocation.SampledRotation)
     Channel=::NewEnumerator2
    )
    Index 0 Param 4=&MergeHelper_FloatSwitch(0=6.6, 1=8.8, 2=5.5, 3=6.6, 4=8.8, 5=5.5, Index=$Particles.SpawnGroup, Num=::NewEnumerator4)
    Module.Index0Float0Write=true
    Module.Index0Float1Write=true
    Module.Index0Float2Write=true
    Module.Index0Float3Write=true
   )
   /SetVariables(Particles.DecalColor:LinearColor=&Multiply_LinearColorByFloat(Linear Color=$Particles.Color, Float=0.15)) {isSetParameters=true}
   &FN_Sprite_Attributes(
    Camera Offset=3
    Color=&MultiplyLinearColor(
     Linear Color=$Particles.Color
     Scale Factor=&MakeVector4FromVectorAndFloat(
      "Vector (RGB)"=<1 1 1>
      "Float (Alpha)"=&FN_FrontEndFloatModifier(Gameplay=0.45, Front End=0.325)
     )
    )
    Create Camera Offset Attribute=true
    Create Color Attribute=true
    Create Position Attribute=true
    Create SpriteSize Attribute=true
    Position:NiagaraPosition=$Output.SocketLocation.SampledPosition
    SpriteSize=&MergeHelper_Vector2Switch(0=<3.3 1>, 1=<4.4 1>, 2=<3.3 1>, 3=<3.3 1>, 4=<4.4 1>, 5=<3.3 1>, Index=$Particles.SpawnGroup, Num=::NewEnumerator4)
    Write to Module Namespace=true
   )
  ]
  rendererClasses [&NiagaraSpriteRendererProperties]
  renderers [
   {rendererClass=&NiagaraSpriteRendererProperties}
   {rendererClass=&NiagaraSpriteRendererProperties, rendererIndex=1}
  ]
 }
}
summary {
 systemName=NS_Backpack_CozyMechanic_Reactivity
 emitters [
  {emitterName=Fans, rendererClasses [&NiagaraMeshRendererProperties]}
  {emitterName=InnerGlow, rendererClasses [&NiagaraSpriteRendererProperties]}
  {emitterName=EyeGlow, rendererClasses [&NiagaraSpriteRendererProperties]}
 ]
 userVariables [
  User.bIsFrontEnd {type=NiagaraBool}
  User.bIsFrontEndPreview {type=NiagaraBool}
  User.ReactivityDebug {type=NiagaraBool}
  User.ReactivityTime {type=NiagaraFloat, defaultValue=0, description=""}
  User.ShouldFadeFE {type=NiagaraBool, defaultValue=true}
  User.Variant {type=NiagaraFloat, defaultValue=0, description=""}
 ]
}
